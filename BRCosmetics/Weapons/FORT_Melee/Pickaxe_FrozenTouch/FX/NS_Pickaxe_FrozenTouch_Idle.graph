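# Niagara idle effect for the Frozen Touch pickaxe that emits drifting ember and mist sprites from randomly chosen weapon sockets.
# Both emitters stop spawning while User.bIsSwinging is set, and Mist also hides existing particles.
# Particles re-sample socket velocity each update to partly follow the moving pickaxe, under gravity, drag and curl noise.
# Colors switch by cosmetic variant, motion is retuned for front end, and Mist adds flipbook animation and time-of-day tinting.
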NiagaraSystem /BRCosmetics/Weapons/FORT_Melee/Pickaxe_FrozenTouch/FX/NS_Pickaxe_FrozenTouch_Idle
dependencies {
 usedDataInterfaces [
  &NiagaraDataInterfaceArrayColor
  &NiagaraDataInterfaceCurve
  &NiagaraDataInterfaceSpriteRendererInfo
  &NiagaraDataInterfaceSocketReader
 ]
 usedDynamicInputs [
  &Vector2DFromFloat
  &RotateVectorByQuaternion
  &FNTimeOfDayScalar
 ]
 usedModules [
  &self:SystemSpawnScript.SystemScriptSource.SystemScriptGraph.NiagaraNodeAssignment_1.SetVariables_#1
  &SystemState
  &self:SystemSpawnScript.SystemScriptSource.SystemScriptGraph.NiagaraNodeAssignment_3.SetVariables_#2
  &FN_VariantManager
 ]
}
emitters {
 Embers {
  emitterUpdateScript [
   &EmitterState()
   &SpawnRate(
    Spawn Probability=&MakeCustomFloatFromBool(Bool=$User.bIsSwinging, True Float=0, False Float=0.6)
    SpawnRate=50
    Use Spawn Probability=true
   )
  ]
  particleSpawnScript [
   /SetVariables(Particles.PerParticleRandom:NiagaraFloat=&RandomRangeFloat(Minimum=0, Maximum=1)) {isSetParameters=true}
   &FN_SetSpawnGroup(Number of Spawn Groups=6)
   &InitializeParticle(
    Color=&VariantColorSelector(
     Variant Amount=4
     Original Color=&RandomRangeLinearColor(Minimum=<rgba 0.2306 0.6423 1 0.1>, Maximum=<rgba 0.1379 0.5198 0.7812 0.2>)
     Variant 1 Color=&RandomRangeLinearColor(Minimum=<rgba 0.3636 0.5227 1 0.2>, Maximum=<rgba 0.6364 0.4059 1 0.3>)
     Variant 2 Color=&RandomRangeLinearColor(Minimum=<rgba 0.9 0.4008 0.1137 0.2>, Maximum=<rgba 1 0.757 0.3646 0.3>)
     Variant 3 Color=&SelectLinearColorFromArray(
      Array Sampling Mode=::NewEnumerator0
      Direct Array Index=?
      Color Selection Array=DataInterface() {propertyValues {ColorData [<rgba 0.9216 0.347 1 0.3>, <rgba 0.2658 0.1457 1 0.3>, <rgba 0.202 0.4019 1 0.3>, <rgba 0.0896 0.8566 1 0.3>]}}
      Random Seed=0
     )
     Variant 4 Color=&RandomRangeLinearColor(Minimum=<rgba 0.2 0.6 1 0.4>, Maximum=<rgba 0.1425 0.4117 0.95 0.5>)
     Variant 5 Color=&RandomRangeLinearColor(Minimum=<rgba 0 0.1471 0.6458 0.5>, Maximum=<rgba 0.2585 0.0117 0.3854 0.6>)
    )
    Lifetime=&RandomRangeFloat(Minimum=0.6, Maximum=1)
    Sprite Size Max=<0.6 0.6>
    Sprite Size Min=<0.4 0.4>
    Sprite Size Mode=::NewEnumerator2
    UsePositionOffset=false
    Write Material Random=true
   )
   /SetVariables(
    Particles.RandomDirection:Vector3f=&RandomVector(Vector Scale=1)
    Particles.RandomSocket:NiagaraInt32=&RandomRangeInt(Minimum=2, Maximum=5)
   ) {isSetParameters=true}
   &SocketLocation(
    Interpolate Socket Sampling=true
    SamplePosOffset=false
    Socket Index=$Particles.RandomSocket
    Socket Reader=$System.Socket Reader
    Socket Sampling Mode=::NewEnumerator3
    VelocitySpeedLimit=false
   )
   &ShapeLocation(MaskSpawnGroup=false, Override Seed=false, Sphere Radius=3, UseOwnerScale=true)
   &AddVelocity(
    Velocity=&RandomVector(Vector Scale=1)
    Velocity Speed Scale=&RandomRangeFloat(Minimum=10, Maximum=20)
   )
  ]
  particleUpdateScript [
   &ParticleState()
   &GravityForce(Gravity=<0 0 -100>)
   &AerodynamicDrag(
    Aerodynamic Drag:NiagaraFloat=&FN_FrontEndFloatModifier(Gameplay=0.2, Front End=0.5)
    Drag Shape Curve=DataInterface() {propertyValues {Curve {keys [
        {interpMode=RCIM_Cubic, tangentMode=RCTM_User, time=0, value=0.1}
        {arriveTangent=2.0799, interpMode=RCIM_Cubic, leaveTangent=2.0799, tangentMode=RCTM_User, time=1, value=1}
       ]}}}
    Lift Contribution:NiagaraFloat=&FN_FrontEndFloatModifier(Gameplay=0.5, Front End=0)
    Lift Shape Curve=DataInterface() {propertyValues {Curve {keys [
        {interpMode=RCIM_Cubic, tangentMode=RCTM_User, time=0, value=0}
        {interpMode=RCIM_Cubic, tangentMode=RCTM_User, time=0.5, value=1}
        {tangentMode=RCTM_User, time=1, value=0}
       ]}}}
    Use Linear Drag=true
    Use Rotational Drag=true
   )
   &CurlNoiseForce(Bias Noise=false, Noise Frequency=100, Noise Strength=100, Pan Noise=true, Pan Noise Field=<0 0 0.1>)
   &SocketLocation(
    Position Sampling=::NewEnumerator1
    SamplePosOffset=false
    Sampled Velocity Coordinate Space:ENiagaraCoordinateSpace=::Simulation
    Socket Index=$Particles.RandomSocket
    Socket Reader=$System.Socket Reader
    Socket Sampling Mode=::NewEnumerator3
    VelocitySpeedLimit=false
   ) {module=SocketLocation001}
   &InheritSourceMovement(
    Applied Position Scale=&VectorFromFloat(Value=&FloatFromCurve(FloatCurve=DataInterface() {propertyValues {Curve {
         defaultValue=1
         keys [{interpMode=RCIM_Cubic, tangentMode=RCTM_User, time=0, value=1}, {interpMode=RCIM_Cubic, tangentMode=RCTM_User, time=1, value=0.8}]
        }}}))
    ApplyForce=false
    ApplyPosition=true
    ApplyPositionFalloff=false
    Source Velocity:Vector3f=$Output.SocketLocation001.SampledVelocity
   )
   &SolveForcesAndVelocity(Clamp Velocity=false, Limit Acceleration=false)
   &ScaleColor(
    Scale Alpha=&FloatFromCurve(
     FloatCurve=DataInterface() {propertyValues {Curve {
        defaultValue=0
        keys [
         {arriveTangent=4.0686, interpMode=RCIM_Cubic, leaveTangent=4.0686, tangentMode=RCTM_User, time=0, value=0}
         {interpMode=RCIM_Cubic, tangentMode=RCTM_User, time=0.5, value=1}
         {tangentMode=RCTM_User, time=1, value=0}
        ]
       }}}
     Scale Curve=&FN_FrontEndFloatModifier(Gameplay=1, Front End=0.8)
    )
    ScaleA=true
    ScaleRGB=true
   )
   &ScaleSpriteSize(
    Uniform Curve Scale=&RandomRangeFloat(Minimum=1, Maximum=&FN_FrontEndFloatModifier(Gameplay=2.5, Front End=1.9))
    Uniform Curve Sprite Scale=DataInterface() {propertyValues {Curve {keys [
        {tangentMode=RCTM_User, time=0, value=1}
        {arriveTangent=-0.0405, interpMode=RCIM_Cubic, leaveTangent=-0.0405, tangentMode=RCTM_User, time=0.7009, value=1}
        {arriveTangent=-8.0089, interpMode=RCIM_Cubic, leaveTangent=-8.0089, tangentMode=RCTM_User, time=1, value=0}
       ]}}}
   )
   /SetVariables(Particles.SpriteSizeGlow:Vector2f=&Multiply_Vector2D_ByFloat(Float=1.2, Vector2D=$Particles.SpriteSize)) {isSetParameters=true}
   &AlignSpriteToMeshOrientation(SetAlignment=true, SetFacing=true)
   /SetVariables(Particles.Glow Color=&MultiplyLinearColor(Linear Color=$Particles.Color, Scale Factor=<2 2 2 0.15>)) {isSetParameters=true}
  ]
  rendererClasses [&NiagaraSpriteRendererProperties]
  renderers [
   {rendererClass=&NiagaraSpriteRendererProperties}
   {rendererClass=&NiagaraSpriteRendererProperties, rendererIndex=1}
  ]
 }
 Mist {
  emitterUpdateScript [
   &EmitterState()
   &SpawnRate(
    Spawn Probability=&MakeCustomFloatFromBool(Bool=$User.bIsSwinging, True Float=0, False Float=0.6)
    SpawnRate=20
    Use Spawn Probability=true
   )
  ]
  particleSpawnScript [
   &FN_SetSpawnGroup(Number of Spawn Groups=6)
   &InitializeParticle(
    Color=&VariantColorSelector(
     Variant Amount=4
     Original Color=&RandomRangeLinearColor(Minimum=<rgba 0.2306 0.6423 1 0.1>, Maximum=<rgba 0.1379 0.5198 0.7812 0.2>)
     Variant 1 Color=&RandomRangeLinearColor(Minimum=<rgba 0.3636 0.5227 1 0.4>, Maximum=<rgba 0.6364 0.4059 1 0.5>)
     Variant 2 Color=&RandomRangeLinearColor(Minimum=<rgba 0.9 0.5712 0.1137 0.2>, Maximum=<rgba 1 0.757 0.3646 0.3>)
     Variant 3 Color=&SelectLinearColorFromArray(
      Array Sampling Mode=::NewEnumerator0
      Direct Array Index=?
      Color Selection Array=DataInterface() {propertyValues {ColorData [<rgba 0.9216 0.347 1 0.35>, <rgba 0.2658 0.1457 1 0.35>, <rgba 0.202 0.4019 1 0.35>, <rgba 0.0896 0.8566 1 0.35>]}}
      Random Seed=0
     )
     Variant 4 Color=&RandomRangeLinearColor(Minimum=<rgba 0.05 0.525 1 0.4>, Maximum=<rgba 0.0475 0.3483 0.95 0.5>)
     Variant 5 Color=&RandomRangeLinearColor(Minimum=<rgba 0 0.1471 0.6458 0.5>, Maximum=<rgba 0.2585 0.0117 0.3854 0.6>)
    )
    Lifetime=&RandomRangeFloat(Minimum=1.75, Maximum=2)
    Sprite Rotation Mode=::NewEnumerator2
    Sprite Size Mode=::NewEnumerator1
    Sprite UV Mode=::NewEnumerator3
    Uniform Sprite Size Max=15
    Uniform Sprite Size Min=12
    UsePositionOffset=false
    Write Material Random=true
   )
   &CameraOffset(Camera Offset Amount=10)
   /SetVariables(
    Particles.RandomDirection:Vector3f=&RandomVector(Vector Scale=1)
    Particles.RandomSocket:NiagaraInt32=&RandomRangeInt(Minimum=2, Maximum=5)
   ) {isSetParameters=true}
   &SocketLocation(
    Interpolate Socket Sampling=true
    SamplePosOffset=true
    Sampled Position Offset=&RandomRangeVector(Minimum=<-1 0 -8>, Maximum=<1 0 8>)
    Socket Index=$Particles.RandomSocket
    Socket Reader=$System.Socket Reader
    Socket Sampling Mode=::NewEnumerator3
    VelocitySpeedLimit=false
   )
   &ShapeLocation(MaskSpawnGroup=false, Override Seed=false, Sphere Radius=1, UseOwnerScale=true)
  ]
  particleUpdateScript [
   &ParticleState()
   &ScaleSpriteSize(Uniform Curve Sprite Scale=DataInterface() {propertyValues {Curve {keys [
        {arriveTangent=0.456, interpMode=RCIM_Cubic, leaveTangent=0.456, tangentMode=RCTM_User, time=0, value=0.7}
        {arriveTangent=0.0089, tangentMode=RCTM_User, time=1, value=1}
       ]}}})
   &ScaleColor(
    Scale Alpha=&FloatFromCurve(
     FloatCurve=DataInterface() {propertyValues {Curve {
        defaultValue=0
        keys [
         {arriveTangent=4.7284, interpMode=RCIM_Cubic, leaveTangent=4.7284, tangentMode=RCTM_User, time=0, value=0}
         {interpMode=RCIM_Cubic, tangentMode=RCTM_User, time=0.4, value=1}
         {tangentMode=RCTM_User, time=1, value=0}
        ]
       }}}
     Scale Curve=&MakeCustomFloatFromBool(
      Bool=$User.bIsSwinging
      True Float=0
      False Float=&FN_FrontEndFloatModifier(Gameplay=0.3, Front End=0.3)
     )
    )
    ScaleA=true
    ScaleRGB=true
   )
   &ScaleColor(
    Color Value To Scale=$Particles.Color
    Scale Mode=::NewEnumerator1
    Scale RGBA=&MakeVector4_FromLinearColor(LinearColor=&FNTimeOfDayColor(
      Night Color=<rgba 0.678 0.7864 1 1>
      Morning Color=<rgba 1 0.9436 0.8834 1>
      Day Color=<rgba 1 1 1 1>
      Evening Color=<rgba 1 0.8812 0.8698 1>
     ))
    ScaleRGBA=true
   ) {module=ScaleColor001}
   &DynamicMaterialParameters(
    Index 0 Param 1=2
    Index 0 Param 2=1
    Index 0 Param 3=1
    Module.Index0Float0Write=true
    Module.Index0Float1Write=true
    Module.Index0Float2Write=true
    Module.Index0Float3Write=true
   )
   &SocketLocation(
    Position Sampling=::NewEnumerator1
    SamplePosOffset=false
    Sampled Velocity Coordinate Space:ENiagaraCoordinateSpace=::Simulation
    Socket Index=$Particles.RandomSocket
    Socket Reader=$System.Socket Reader
    Socket Sampling Mode=::NewEnumerator3
    VelocitySpeedLimit=false
   ) {module=SocketLocation001}
   &GravityForce(Gravity=&Multiply_VectorByFloat(
     Vector=&RandomRangeVector(Minimum=<0 0 -300>, Maximum=<0 0 -500>)
     Float=&FN_FrontEndFloatModifier(Gameplay=1, Front End=0.5)
    ))
   &AerodynamicDrag(
    Aerodynamic Drag:NiagaraFloat=&FN_FrontEndFloatModifier(Gameplay=1, Front End=2)
    Drag Shape Curve=DataInterface() {propertyValues {Curve {keys [
        {interpMode=RCIM_Cubic, tangentMode=RCTM_User, time=0, value=0.1}
        {arriveTangent=2.0799, interpMode=RCIM_Cubic, leaveTangent=2.0799, tangentMode=RCTM_User, time=1, value=1}
       ]}}}
    Lift Contribution:NiagaraFloat=&FN_FrontEndFloatModifier(Gameplay=0, Front End=0.5)
    Lift Shape Curve=DataInterface() {propertyValues {Curve {keys [
        {interpMode=RCIM_Cubic, tangentMode=RCTM_User, time=0, value=0}
        {interpMode=RCIM_Cubic, tangentMode=RCTM_User, time=0.5, value=1}
        {tangentMode=RCTM_User, time=1, value=0}
       ]}}}
    Use Linear Drag=true
    Use Rotational Drag=true
   )
   &CurlNoiseForce(Bias Noise=false, Noise Frequency=100, Noise Strength=100, Pan Noise=true, Pan Noise Field=<0 0 -0.5>)
   &InheritSourceMovement(
    Applied Position Scale=&VectorFromFloat(Value=&FloatFromCurve(FloatCurve=DataInterface() {propertyValues {Curve {
         defaultValue=1
         keys [
          {interpMode=RCIM_Cubic, tangentMode=RCTM_User, time=0, value=1}
          {arriveTangent=-0.4114, interpMode=RCIM_Cubic, leaveTangent=-0.4114, tangentMode=RCTM_User, time=1, value=0.85}
         ]
        }}}))
    ApplyForce=false
    ApplyPosition=true
    ApplyPositionFalloff=false
    Source Velocity:Vector3f=$Output.SocketLocation001.SampledVelocity
   )
   &SolveForcesAndVelocity(Clamp Velocity=false, Limit Acceleration=false)
   &AlignSpriteToMeshOrientation(SetAlignment=true, SetFacing=true)
   &SubUVAnimation(
    "Sprite Renderer "=DataInterface() {propertyValues {SpriteRenderer=&self:Mist_0.NiagaraSpriteRendererProperties_0}}
    Start Frame Offset:NiagaraInt32=&RandomRangeInt(Minimum=0, Maximum=20)
    SubUV Lookup Index Scale=0.25
    UseEndFrame=false
    UseStartFrame=false
   )
  ]
  rendererClasses [&NiagaraSpriteRendererProperties]
  renderers [{rendererClass=&NiagaraSpriteRendererProperties}]
 }
}
summary {
 systemName=NS_Pickaxe_FrozenTouch_Idle
 emitters [
  {emitterName=Mist, rendererClasses [&NiagaraSpriteRendererProperties]}
  {emitterName=Embers, rendererClasses [&NiagaraSpriteRendererProperties]}
 ]
 userVariables [
  User.bIsFrontEnd {type=NiagaraBool}
  User.bIsFrontEndPreview {type=NiagaraBool}
  User.bIsSwinging {type=NiagaraBool}
  User.Variant {type=NiagaraFloat, defaultValue=0, description=""}
 ]
}
