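# Niagara impact burst for the Frozen Touch pickaxe, aimed along a system direction vector.
# One-shot emitters layer flashes with a light, ice shard meshes, mist, sparks, spike streaks and variant elements.
# Variant selectors switch color palettes and variant element size, spread and rotation, while the variant index also gates ice shard visibility.
# Front-end modifiers mostly shrink the effect outside gameplay, and time-of-day color and brightness inputs tint many layers.
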NiagaraSystem /BRCosmetics/Weapons/FORT_Melee/Pickaxe_FrozenTouch/FX/NS_Pickaxe_FrozenTouch_Impact
dependencies {
 usedDataInterfaces [
  &NiagaraDataInterfaceArrayColor
  &NiagaraDataInterfaceCurve
  &NiagaraDataInterfaceSpriteRendererInfo
  &NiagaraDataInterfaceVector2DCurve
  &NiagaraDataInterfaceColorCurve
  &NiagaraDataInterfaceArrayFloat
  &NiagaraDataInterfaceArrayInt32
  &NiagaraDataInterfaceArrayFloat3
 ]
 usedDynamicInputs [
  &MergeHelper_VectorSwitch
  &MaskVectorBySpawnGroup
  &SelectVectorFromArray
  &SetBoolByFloatComparison
 ]
 usedModules [
  &SystemState
  &self:SystemSpawnScript.SystemScriptSource.SystemScriptGraph.NiagaraNodeAssignment_1.SetVariables_#1
  &self:SystemSpawnScript.SystemScriptSource.SystemScriptGraph.NiagaraNodeAssignment_0.SetVariables_#2
  &FN_VariantManager
  &self:SystemSpawnScript.SystemScriptSource.SystemScriptGraph.NiagaraNodeAssignment_2.SetVariables_#3
  &self:SystemSpawnScript.SystemScriptSource.SystemScriptGraph.NiagaraNodeAssignment_3.SetVariables_#4
 ]
}
emitters {
 Directional_Flash {
  emitterUpdateScript [
   &EmitterState()
   &SpawnBurst_Instantaneous(Use Spawn Probability=false, UseLoopCountLimit=false)
  ]
  particleSpawnScript [
   &InitializeParticle(
    Color=<rgba 0.0073 0.0085 0.01 0.25>
    Lifetime=0.15
    Sprite Size=&Multiply_Vector2D_ByFloat(
     Float=&FN_FrontEndFloatModifier(Gameplay=1, Front End=0.65)
     Vector2D=&Multiply_Vector2D_ByFloat(Float=1, Vector2D=<50 80>)
    )
    Sprite Size Mode=::NewEnumerator4
    UsePositionOffset=false
    Write Material Random=true
   )
   &CameraOffset(Camera Offset Amount=25)
   /SetVariables(Particles.CoreSprite:Vector2f=&Multiply_Vector2D_ByFloat(Float=&FN_FrontEndFloatModifier(Gameplay=1, Front End=0.6), Vector2D=<90 150>)) {isSetParameters=true}
   &SpriteFacingAndAlignment(Alignment Coordinate Space=::Local, Sprite Alignment=$System.DirectionVector, WriteAlignment=true, WriteFacing=false)
   &AddVelocity(Velocity=$System.DirectionVector, Velocity Speed Scale=30)
  ]
  particleUpdateScript [
   &ParticleState()
   &ScaleSpriteSize(
    Non-Uniform Curve Scale=<1.2 1.2>
    Non-Uniform Curve Sprite Scale=DataInterface() {propertyValues {
      XCurve {
       defaultValue=0.35
       keys [
        {interpMode=RCIM_Cubic, tangentMode=RCTM_User, time=0, value=0}
        {interpMode=RCIM_Cubic, tangentMode=RCTM_User, time=0.5, value=1}
        {tangentMode=RCTM_User, time=1, value=0.5}
       ]
      }
      YCurve {keys [{interpMode=RCIM_Cubic, leaveTangent=2.5591, leaveTangentWeight=0.1667, tangentMode=RCTM_Break, time=0, value=0.1}, <at 1 1.2>]}
     }}
    Scale Sprite Size Mode=::NewEnumerator3
   )
   /SetVariables(
    Particles.CoreColor:LinearColor=&Multiply_LinearColorByFloat(
     Linear Color=&VariantColorSelector(
      Variant Amount=4
      Original Color=<rgba 0.7187 1.9883 3 0.3>
      Variant 1 Color=<rgba 1.2 1.35 3 0.3>
      Variant 2 Color=<rgba 3 1.8603 1.0463 0.3>
      Variant 3 Color=&SelectLinearColorFromArray(
       Array Sampling Mode=::NewEnumerator0
       Direct Array Index=?
       Color Selection Array=DataInterface() {propertyValues {ColorData [<rgba 2.5049 1.2 3 0.3>, <rgba 0.75 1.1805 3 0.3>, <rgba 1.05 2.7457 3 0.3>, <rgba 3 1.2 2.4209 0.3>]}}
       Random Seed=0
      )
      Variant 4 Color=<rgba 0.45 1.725 3 0.3>
      Variant 5 Color=<rgba 2.1617 0.3998 4 0.3>
     )
     Float=$System.TODFloat
    )
    Particles.CoreSprite:Vector2f=&Multiply_Vector2D(
     A=$Particles.Initial.CoreSprite
     B=&Vector2DFromCurve(
      Vector2Curve=DataInterface() {propertyValues {
        XCurve {
         defaultValue=0.1
         keys [
          {interpMode=RCIM_Cubic, tangentMode=RCTM_User, time=0, value=0}
          {interpMode=RCIM_Cubic, tangentMode=RCTM_User, time=0.5, value=1}
          {tangentMode=RCTM_User, time=1, value=0.5}
         ]
        }
        YCurve {keys [{interpMode=RCIM_Cubic, leaveTangent=4.0442, tangentMode=RCTM_Break, time=0, value=0.1}, {tangentMode=RCTM_User, time=0.5, value=0.95}]}
       }}
      ScaleCurve=<1 1>
     )
    )
   ) {isSetParameters=true}
   &Drag(Use Linear Drag=true, Use Rotational Drag=true)
   &SolveForcesAndVelocity(Clamp Velocity=false, Limit Acceleration=false)
   &DynamicMaterialParameters(
    Index 0 Param 1=&FloatFromCurve(FloatCurve=DataInterface() {propertyValues {Curve {keys [
         {arriveTangent=1.5862, interpMode=RCIM_Cubic, leaveTangent=1.5862, tangentMode=RCTM_User, time=0, value=0}
         {arriveTangent=0.159, interpMode=RCIM_Cubic, leaveTangent=0.159, tangentMode=RCTM_User, time=1, value=0.8}
        ]}}})
    Index 0 Param 2=&FloatFromCurve(
     FloatCurve=DataInterface() {propertyValues {Curve {keys [
         {arriveTangent=2.5938, interpMode=RCIM_Cubic, leaveTangent=2.5938, tangentMode=RCTM_User, time=0, value=0.2}
         {interpMode=RCIM_Cubic, tangentMode=RCTM_User, time=0.2, value=0.5}
         {tangentMode=RCTM_User, time=1, value=0}
        ]}}}
     Scale Curve=0.7
    )
    Index 0 Param 3=&FloatFromCurve(
     FloatCurve=DataInterface() {propertyValues {Curve {keys [{interpMode=RCIM_Cubic, tangentMode=RCTM_User, time=0, value=1}, {arriveTangent=-0.3137, tangentMode=RCTM_User, time=1, value=0.8}]}}}
     Scale Curve=1.5
    )
    Index 0 Param 4=&Add_Float(
     A=&RandomRangeFloat(Minimum=0, Maximum=0.25)
     B=&FloatFromCurve(
      FloatCurve=DataInterface() {propertyValues {Curve {keys [{interpMode=RCIM_Cubic, tangentMode=RCTM_User, time=0, value=0}, {interpMode=RCIM_Cubic, tangentMode=RCTM_User, time=1, value=1}]}}}
      Scale Curve=0.25
     )
    )
    Module.Param0WriteEnabled=true
    Module.Param1WriteEnabled=true
    Module.Param2WriteEnabled=true
    Module.Param3WriteEnabled=true
   )
  ]
  rendererClasses [&NiagaraSpriteRendererProperties]
  renderers [
   {rendererClass=&NiagaraSpriteRendererProperties}
   {rendererClass=&NiagaraSpriteRendererProperties, rendererIndex=1}
  ]
 }
 FlashElements {
  emitterUpdateScript [
   &EmitterState()
   &SpawnBurst_Instantaneous(Spawn Group=1, Use Spawn Probability=false, UseLoopCountLimit=false) {module=SpawnBurst_Instantaneous001}
   &SpawnBurst_Instantaneous(
    Spawn Count=&RandomRangeInt(Minimum=6, Maximum=8)
    Use Spawn Probability=false
    UseLoopCountLimit=false
   )
  ]
  particleSpawnScript [
   &InitializeParticle(
    Color Mode=::NewEnumerator0
    Lifetime=0.3
    Sprite Rotation Angle=0.25
    Sprite Rotation Mode=::NewEnumerator1
    UsePositionOffset=false
    Write Material Random=true
   )
   &ShapeLocation(MaskSpawnGroup=false, Override Seed=false, Sphere Radius=1, UseOwnerScale=true)
   &SpriteFacingAndAlignment(Sprite Alignment=$System.DirectionVector, WriteAlignment=true, WriteFacing=false)
   &FN_PerParticleLOD(Enable SpriteSize Distance Scaling=true, Sprite Size Scale Far=<2.5 2.5>)
   /SetVariables(Particles.SpriteSize=<0.8 1.5>, Particles.VisibilityTag=&SpawnGroup()) {isSetParameters=true}
   /SetVariables(
    Particles.Flash Spike Lifetime:NiagaraFloat=&RandomRangeFloat(Minimum=0.05, Maximum=0.07)
    Particles.Flash Spike Sprite Size:Vector2f=&FN_FrontEndVector2Modifier(
     Front End:Vector2f=&RandomRangeVector2D(Minimum=<2 25>, Maximum=<7 40>)
     Gameplay:Vector2f=&RandomRangeVector2D(Minimum=<3 55>, Maximum=<10 70>)
    )
   ) {isSetParameters=true}
   /SetVariables(
    Particles.Flash Star Cam Offset=15
    Particles.Flash Star Lifetime=0.16
    Particles.Flash Star Sprite Size:Vector2f=&Multiply_Vector2D_ByFloat(
     Float=&Multiply_Float(
      A=&MakeCustomFloatFromBool(Bool=$User.bIsFrontEnd, True Float=0.6, False Float=1)
      B=&RandomRangeFloat(Minimum=140, Maximum=150)
     )
     Vector2D=$Particles.SpriteSize
    )
    Particles.Flash Star Sprite Size Alt:Vector2f=&Multiply_Vector2D_ByFloat(
     Float=&Multiply_Float(
      A=0.85
      B=&Multiply_Float(
       A=&MakeCustomFloatFromBool(Bool=$User.bIsFrontEnd, True Float=0.6, False Float=1)
       B=&RandomRangeFloat(Minimum=120, Maximum=150)
      )
     )
     Vector2D=$Particles.SpriteSize
    )
    Particles.Flash Star SubUV Index:NiagaraFloat=&RandomRangeFloat(Minimum=0, Maximum=3)
    Particles.Flash Star SubUV Index Alt:NiagaraFloat=&RandomRangeFloat(Minimum=0, Maximum=3)
   ) {isSetParameters=true}
   /SetVariables(
    Particles.Flash Flare Color:LinearColor=&VariantColorSelector(
     Variant Amount=4
     Original Color=<rgba 0.4115 0.8868 1 0.7>
     Variant 1 Color=<rgba 0.7595 0.4115 1 0.7>
     Variant 2 Color=<rgba 1 0.6086 0.3281 0.7>
     Variant 3 Color=<rgba 0.9828 0.6024 1 0.7>
     Variant 4 Color=<rgba 0.05 0.525 1 0.7>
     Variant 5 Color=<rgba 0.1237 0.1237 1.25 0.7>
    )
    Particles.Flash Flare Lifetime:NiagaraFloat=&FN_FrontEndFloatModifier(Gameplay=0.2, Front End=0.125)
    Particles.Flash Flare Sprite Size:Vector2f=&FN_FrontEndVector2Modifier(Front End:Vector2f=<130 130>, Gameplay:Vector2f=<300 200>)
   ) {isSetParameters=true}
   &DynamicMaterialParameters(
    Index 0 Param 1=0.2
    Index 0 Param 2=1
    Module.Index0Param0Write=true
    Module.Index0Param1Write=true
    Module.Index0Param2Write=true
    Module.Index0Param3Write=true
   )
   &CameraOffset(Camera Offset Amount=12)
  ]
  particleUpdateScript [
   &ParticleState()
   /SetVariables(
    Particles.Update Flash Flare Age:NiagaraFloat=&Divide_Float(A=$Particles.Age, B=$Particles.Initial.Flash Flare Lifetime)
    Particles.Update Flash Spike Age:NiagaraFloat=&Divide_Float(A=$Particles.Age, B=$Particles.Initial.Flash Spike Lifetime)
    Particles.Update Flash Star Age:NiagaraFloat=&Divide_Float(A=$Particles.Age, B=$Particles.Initial.Flash Star Lifetime)
   ) {isSetParameters=true}
   /SetVariables(
    Particles.Update Flash Spike Color:LinearColor=&MakeLinearColorFromVectorAndFloat(
     "Vector (RGB)"=&MakeVectorFromLinearColorRGB(Color=$Particles.Initial.Flash Spike Color)
     "Float (Alpha)"=&FloatFromCurve(FloatCurve=DataInterface() {propertyValues {Curve {keys [<at 0.7 1>,<at 1 0>]}}}, CurveIndex=$Particles.Update Flash Spike Age)
    )
    Particles.Update Flash Spike Sprite Size:Vector2f=&Multiply_Vector2D_ByFloat(
     Float=&FloatFromCurve(
      FloatCurve=DataInterface() {propertyValues {Curve {keys [<at 0 1.2>,<at 1 2>]}}}
      CurveIndex=$Particles.Update Flash Spike Age
      Scale Curve=1.25
     )
     Vector2D=$Particles.Initial.Flash Spike Sprite Size
    )
    Particles.Update Flash Star Color:LinearColor=&Multiply_LinearColorByFloat(
     Linear Color=&MultiplyLinearColor(
      Linear Color=&VariantColorSelector(
       Variant Amount=4
       Original Color=&ColorFromCurve(
        ColorCurve=DataInterface() {propertyValues {
          AlphaCurve {keys [<at 0.0941 1>,<at 0.8992 0>]}
          BlueCurve {keys [<at 0.1978 3>,<at 0.4068 2>,<at 0.6127 1.5>]}
          GreenCurve {keys [<at 0.1978 2.3918>,<at 0.4068 1.4551>,<at 0.6127 1.1207>]}
          RedCurve {keys [<at 0.1978 0.7403>,<at 0.4068 0.4389>,<at 0.6127 0.233>]}
         }}
        CurveIndex=$Particles.Update Flash Star Age
       )
       Variant 1 Color=&ColorFromCurve(
        ColorCurve=DataInterface() {propertyValues {
          AlphaCurve {keys [<at 0.0928 1>,<at 0.8992 0>]}
          BlueCurve {keys [<at 0.1978 3>,<at 0.4068 2>,<at 0.6914 1.5>]}
          GreenCurve {keys [<at 0.1978 0.7403>,<at 0.4068 0.4389>,<at 0.6914 0.233>]}
          RedCurve {keys [<at 0.1978 1.7228>,<at 0.4068 0.6289>,<at 0.6914 0.5194>]}
         }}
        CurveIndex=$Particles.Update Flash Star Age
       )
       Variant 2 Color=&ColorFromCurve(
        ColorCurve=DataInterface() {propertyValues {
          AlphaCurve {keys [<at 0.0877 1>,<at 0.8992 0>]}
          BlueCurve {keys [<at 0.1978 0.7403>,<at 0.4068 0.4389>,<at 0.6914 0.233>]}
          GreenCurve {keys [<at 0.1978 1.8702>,<at 0.4068 1.0893>,<at 0.6914 0.6553>]}
          RedCurve {keys [<at 0.1978 3>,<at 0.4068 2>,<at 0.6914 1.5>]}
         }}
        CurveIndex=$Particles.Update Flash Star Age
       )
       Variant 3 Color=&ColorFromCurve(
        ColorCurve=DataInterface() {propertyValues {
          AlphaCurve {keys [<at 0.0967 1>,<at 0.8992 0>]}
          BlueCurve {keys [<at 0.1121 3>,<at 0.452 2>,<at 0.9775 1.5>]}
          GreenCurve {keys [<at 0.1121 0.5217>,<at 0.452 1.8448>,<at 0.9775 0.1426>]}
          RedCurve {keys [<at 0.1121 2.427>,<at 0.452 0>,<at 0.9775 0.042>]}
         }}
        CurveIndex=$Particles.Update Flash Star Age
       )
       Variant 4 Color=&ColorFromCurve(
        ColorCurve=DataInterface() {propertyValues {
          AlphaCurve {keys [<at 0.1057 1>,<at 0.8992 0>]}
          BlueCurve {keys [<at 0.1054 3>,<at 0.4929 2>,<at 0.906 1.5>]}
          GreenCurve {keys [<at 0.1054 1.3>,<at 0.4929 1.2194>,<at 0.906 0.4442>]}
          RedCurve {keys [<at 0.1054 0.45>,<at 0.4929 0.4389>,<at 0.906 0.233>]}
         }}
        CurveIndex=$Particles.Update Flash Star Age
       )
       Variant 5 Color=&ColorFromCurve(
        ColorCurve=DataInterface() {propertyValues {
          AlphaCurve {keys [<at 0.1082 1>,<at 0.9559 0>]}
          BlueCurve {keys [<at 0 3>,<at 0.7714 0.444>]}
          GreenCurve {keys [<at 0 0.875>,<at 0.7714 0.0301>]}
          RedCurve {keys [<at 0 0.45>,<at 0.7714 0.1236>]}
         }}
        CurveIndex=$Particles.Update Flash Star Age
       )
      )
      Scale Factor=&MakeVector4_FromLinearColor(LinearColor=$System.TODColor)
     )
     Float=$System.TODFloat
    )
    Particles.Update Flash Star Sprize Size:Vector2f=&Multiply_Vector2D_ByFloat(
     Float=&FloatFromCurve(
      FloatCurve=DataInterface() {propertyValues {Curve {keys [
          {arriveTangent=2.7013, interpMode=RCIM_Cubic, leaveTangent=2.7013, tangentMode=RCTM_User, time=0, value=0}
          {interpMode=RCIM_Cubic, tangentMode=RCTM_User, time=1, value=1}
         ]}}}
      CurveIndex=$Particles.Update Flash Star Age
     )
     Vector2D=$Particles.Initial.Flash Star Sprite Size
    )
    Particles.Update Flash Star Sprize Size Alt:Vector2f=&Multiply_Vector2D_ByFloat(
     Float=&FloatFromCurve(
      FloatCurve=DataInterface() {propertyValues {Curve {keys [
          {arriveTangent=2.7013, interpMode=RCIM_Cubic, leaveTangent=2.7013, tangentMode=RCTM_User, time=0, value=0}
          {interpMode=RCIM_Cubic, tangentMode=RCTM_User, time=1, value=1}
         ]}}}
      CurveIndex=$Particles.Update Flash Star Age
      Scale Curve=1.25
     )
     Vector2D=$Particles.Initial.Flash Star Sprite Size Alt
    )
   ) {isSetParameters=true}
   /SetVariables(Particles.Update Flash Flare Color:LinearColor=&MultiplyLinearColor(
     Linear Color=&Multiply_LinearColorByFloat(
      Linear Color=$System.TODColor
      Float=&Multiply_Float(A=$System.TODFloat, B=&FN_FrontEndFloatModifier(Gameplay=1, Front End=0.5))
     )
     Scale Factor=&MakeVector4_FromLinearColor(LinearColor=&MakeLinearColorFromVectorAndFloat(
       "Vector (RGB)"=&Multiply_Vector(
        A=&MakeVectorFromLinearColorRGB(Color=$Particles.Initial.Flash Flare Color)
        B=&VectorFromFloat(Value=&FN_FrontEndFloatModifier(Gameplay=1, Front End=1))
       )
       "Float (Alpha)"=&Multiply_Float(
        A=&MakeFloatFromLinearColor(LinearColor=<rgba 1 1 1 1>, Channel:ENiagaraLinearColor_Channels=::NewEnumerator3)
        B=&Lerp_Float(A=1, B=0, Alpha=$Particles.Update Flash Flare Age)
       )
      ))
    )) {isSetParameters=true}
   &FN_Light_Attributes(
    Color=&Multiply_LinearColorByFloat(
     Linear Color=&VariantColorSelector(
      Variant Amount=4
      Original Color=<rgba 0.2212 0.8363 1 1>
      Variant 1 Color=<rgba 0.2212 0.351 1 1>
      Variant 2 Color=<rgba 1 0.465 0.2212 1>
      Variant 3 Color=<rgba 0.2501 0.1549 1 1>
      Variant 4 Color=<rgba 0 0.5 1 1>
      Variant 5 Color=<rgba 0.0471 0.0194 1 1>
     )
     Float=&FNTimeOfDayScalar(Night Scale=0.8, Morning Scale=0.9, Evening Scale=0.9)
    )
    Create Color Attribute=true
    Create Exponent Attribute=true
    Create Position Attribute=true
    Create Radius Attribute=true
    Exponent=5
    Position=<5 0 0>
    Radius=75
    Scale Color=&FloatFromCurve(
     FloatCurve=DataInterface() {propertyValues {Curve {keys [
         {interpMode=RCIM_Cubic, tangentMode=RCTM_User, time=0, value=1}
         {arriveTangent=-0.5426, interpMode=RCIM_Cubic, leaveTangent=-0.5426, tangentMode=RCTM_User, time=1, value=0}
        ]}}}
     Scale Curve=10
    )
   )
  ]
  rendererClasses [&NiagaraSpriteRendererProperties, &NiagaraLightRendererProperties]
  renderers [
   {rendererClass=&NiagaraSpriteRendererProperties}
   {rendererClass=&NiagaraSpriteRendererProperties, rendererIndex=1}
   {rendererClass=&NiagaraLightRendererProperties, rendererIndex=2}
  ]
 }
 Ice_Shards {
  emitterUpdateScript [
   &EmitterState()
   &SpawnBurst_Instantaneous(
    Spawn Count=&RandomRangeInt(Minimum=4, Maximum=6)
    Use Spawn Probability=false
    UseLoopCountLimit=false
   )
  ]
  particleSpawnScript [
   &FN_SetSpawnGroup(Number of Spawn Groups=2)
   &InitializeParticle(
    Color=&VariantColorSelector(
     Variant Amount=4
     Original Color=<rgba 0.049 0.209 0.422 1>
     Variant 1 Color=<rgba 0.16 0.16 0.64 1>
     Variant 2 Color=<rgba 0.049 0.209 0.422 1>
     Variant 3 Color=&SelectLinearColorFromArray(
      Array Sampling Mode=::NewEnumerator0
      Direct Array Index=?
      Color Selection Array=DataInterface() {propertyValues {ColorData [<rgba 0.5667 0.24 0.8 1>, <rgba 0.2 0.3148 0.8 1>, <rgba 0.224 0.7249 0.8 1>, <rgba 0.8 0.24 0.6198 1>]}}
      Random Seed=0
     )
     Variant 4 Color=<rgba 0 0.25 0.75 1>
     Variant 5 Color=<rgba 1 1 1 1>
    )
    Lifetime Max=0.25
    Lifetime Min=0.2
    Lifetime Mode=::NewEnumerator1
    Mesh Scale=&Multiply_VectorByFloat(
     Vector=&RandomRangeVector(Minimum=<0.4 0.4 0.5>, Maximum=<0.5 0.5 0.8>)
     Float=&FN_FrontEndFloatModifier(Gameplay=1, Front End=0.5)
    )
    Mesh Scale Mode=::NewEnumerator4
    Sprite Size Mode=::NewEnumerator3
    Uniform Sprite Size=&RandomRangeFloat(Minimum=15, Maximum=20)
    UsePositionOffset=false
    Write Material Random=true
   )
   &ShapeLocation(
    Cylinder Height=&Multiply_Float(A=60, B=&FN_FrontEndFloatModifier(Gameplay=1, Front End=0.6))
    Cylinder Height Midpoint=0.1
    Cylinder Radius=6
    MaskSpawnGroup=false
    Override Seed=false
    Rotation Axis=$System.DirectionVector
    Rotation Mode=::NewEnumerator1
    Shape Primitive=::NewEnumerator1
    Surface Only=false
    UseOwnerScale=true
   )
   &AddVelocity(
    Module.Use Falloff Distance=false
    Origin Offset=&Multiply_VectorByFloat(
     Vector=&NormalizeVector(Vector To Normalize=$System.DirectionVector, Safe Normalize=true, Safe Fallback Threshold=0, Safe Fallback Vector=<0 0 0.001>)
     Float=-10
    )
    Override Seed=false
    Velocity Mode=::NewEnumerator1
    Velocity Speed=&RandomRangeFloat(Minimum=10, Maximum=80)
   ) {module=AddVelocity001}
   &InitialMeshOrientation(
    Mesh Axis To Orient=<0 0 1>
    Mesh Orientation Mode=::NewEnumerator2
    Rotation=&RandomRangeVector(Minimum=<0 0 0>, Maximum=<0 0 1>)
    Use Rotation Vector=true
   )
   &CameraOffset(Camera Offset Amount=10)
   /SetVariables(Particles.VisibilityTag=&MergeHelper_IntSwitch(
     0=0
     1=0
     2=1
     3=0
     4=0
     5=1
     Index=&MakeIntFromFloat(Float=$System.VariantIndex)
     Num=::NewEnumerator4
    )) {isSetParameters=true}
  ]
  particleUpdateScript [
   &ParticleState()
   &ScaleColor(
    Scale Alpha=&FloatFromCurve(FloatCurve=DataInterface() {propertyValues {Curve {keys [<at 0.5 1 RCIM_Cubic>,<at 1.0011 0.002 RCIM_Cubic>]}}})
    Scale RGB=&MakeVectorFromLinearColorRGB(Color=&FNTimeOfDayColor(
      Night Color=<rgba 0.6 0.6666 0.8 1>
      Morning Color=<rgba 0.9 0.7671 0.738 1>
      Day Color=<rgba 1 1 1 1>
      Evening Color=<rgba 0.9 0.7742 0.738 1>
     ))
    ScaleA=true
    ScaleRGB=true
   )
   &ScaleMeshSize(Scale Factor=&VectorFromFloat(Value=&FloatFromCurve(
      FloatCurve=DataInterface() {propertyValues {Curve {
         defaultValue=0
         keys [
          {arriveTangent=1.9787, interpMode=RCIM_Cubic, leaveTangent=1.9787, tangentMode=RCTM_User, time=0, value=0.75}
          {arriveTangent=-1.7218, interpMode=RCIM_Cubic, leaveTangent=-1.7218, tangentMode=RCTM_User, time=1, value=0}
         ]
        }}}
      Scale Curve=1.7
     )))
   /SetVariables(Particles.Position=&AddVectorToPosition(
     Position=$Particles.Initial.Position
     Vector=&Multiply_VectorByFloat(
      Vector=$Particles.Velocity
      Float=&FloatFromCurve(
       FloatCurve=DataInterface() {propertyValues {Curve {
          defaultValue=0
          keys [
           {arriveTangent=1.2412, interpMode=RCIM_Cubic, leaveTangent=1.2412, tangentMode=RCTM_User, time=0, value=0}
           {arriveTangent=0.547, interpMode=RCIM_Cubic, leaveTangent=0.547, tangentMode=RCTM_User, time=1, value=1}
          ]
         }}}
       Scale Curve=&FN_FrontEndFloatModifier(Gameplay=1, Front End=0.6)
      )
     )
    )) {isSetParameters=true}
  ]
  rendererClasses [&NiagaraMeshRendererProperties]
  renderers [{rendererClass=&NiagaraMeshRendererProperties, rendererIndex=0}]
 }
 Mist {
  emitterUpdateScript [
   &EmitterState()
   &SpawnBurst_Instantaneous(
    Spawn Count=&RandomRangeInt(Minimum=2, Maximum=3)
    Use Spawn Probability=false
    UseLoopCountLimit=false
   )
  ]
  particleSpawnScript [
   &InitializeParticle(
    Color=&VariantColorSelector(
     Variant Amount=4
     Original Color=&RandomRangeLinearColor(Minimum=<rgba 0.2306 0.6423 1 1>, Maximum=<rgba 0.1379 0.5198 0.7812 1>)
     Variant 1 Color=&RandomRangeLinearColor(Minimum=<rgba 0.3636 0.5227 1 1>, Maximum=<rgba 0.6364 0.4059 1 1>)
     Variant 2 Color=&RandomRangeLinearColor(Minimum=<rgba 0.9 0.5712 0.1137 0.6>, Maximum=<rgba 1 0.757 0.3646 0.8>)
     Variant 3 Color=&SelectLinearColorFromArray(
      Array Sampling Mode=::NewEnumerator0
      Direct Array Index=?
      Color Selection Array=DataInterface() {propertyValues {ColorData [<rgba 0.9216 0.347 1 0.3>, <rgba 0.2658 0.1457 1 0.3>, <rgba 0.202 0.4019 1 0.3>, <rgba 0.0896 0.8566 1 0.3>]}}
      Random Seed=0
     )
     Variant 4 Color=&RandomRangeLinearColor(Minimum=<rgba 0.1304 0.5652 1 1>, Maximum=<rgba 0.16 0.3733 0.8 1>)
     Variant 5 Color=&RandomRangeLinearColor(Minimum=<rgba 0 0.1471 0.6458 1>, Maximum=<rgba 0.2585 0.0117 0.3854 1>)
    )
    Lifetime=&RandomRangeFloat(Minimum=0.8, Maximum=1)
    Sprite Rotation Angle=&RandomRangeFloat(Minimum=0, Maximum=360)
    Sprite Rotation Mode=::NewEnumerator0
    Sprite Size Mode=::NewEnumerator3
    Sprite UV Mode=::NewEnumerator2
    Uniform Sprite Size=&Multiply_Float(
     A=&RandomRangeFloat(Minimum=80, Maximum=100)
     B=&FN_FrontEndFloatModifier(Gameplay=1, Front End=0.55)
    )
    UsePositionOffset=false
    Write Material Random=true
   )
   &AddVelocity(
    Cone Angle=30
    Cone Axis=$System.DirectionVector
    Module.UseConeFalloff=false
    Module.UseVelDistribution=false
    Override Seed=false
    Velocity Mode=::NewEnumerator2
    Velocity Speed=&Multiply_Float(
     A=&RandomRangeFloat(Minimum=100, Maximum=250)
     B=&FN_FrontEndFloatModifier(Gameplay=1, Front End=0.6)
    )
   )
   &ShapeLocation(
    Cylinder Height=&Multiply_Float(A=50, B=&FN_FrontEndFloatModifier(Gameplay=1, Front End=0.6))
    Cylinder Radius=3
    MaskSpawnGroup=false
    Override Seed=false
    Rotation Axis=$System.DirectionVector
    Rotation Mode=::NewEnumerator1
    Shape Primitive=::NewEnumerator1
    Surface Only=false
    UseOwnerScale=true
   )
   &DynamicMaterialParameters(
    Index 0 Param 1=2
    Index 0 Param 2=0.5
    Index 0 Param 3=0.5
    Index 0 Param 4=1
    Module.Index0Float0Write=true
    Module.Index0Float1Write=true
    Module.Index0Float2Write=true
    Module.Index0Float3Write=true
   )
  ]
  particleUpdateScript [
   &ParticleState()
   &GravityForce(Gravity=<0 0 -50>)
   &Drag(Drag=4, Use Linear Drag=true, Use Rotational Drag=true)
   &SolveForcesAndVelocity(Clamp Velocity=false, Limit Acceleration=false)
   &ScaleSpriteSize(
    Uniform Curve Scale=&FN_FrontEndFloatModifier(Gameplay=1, Front End=0.6)
    Uniform Curve Sprite Scale=DataInterface() {propertyValues {Curve {
       defaultValue=0
       keys [
        {arriveTangent=0.8238, interpMode=RCIM_Cubic, leaveTangent=0.8238, tangentMode=RCTM_User, time=0, value=0.55}
        {interpMode=RCIM_Cubic, tangentMode=RCTM_User, time=1, value=1}
       ]
      }}}
   )
   &CameraOffset(Camera Offset Amount=&FloatFromCurve(
     FloatCurve=DataInterface() {propertyValues {Curve {defaultValue=1, keys [{tangentMode=RCTM_User, time=0, value=0}, {tangentMode=RCTM_User, time=1, value=1}]}}}
     Scale Curve=10
    ))
   &ScaleColor(
    Scale Alpha=&FloatFromCurve(
     FloatCurve=DataInterface() {propertyValues {Curve {
        defaultValue=0
        keys [
         {arriveTangent=0.8078, interpMode=RCIM_Cubic, leaveTangent=0.8078, tangentMode=RCTM_User, time=0, value=0.8}
         {arriveTangent=0.0132, interpMode=RCIM_Cubic, leaveTangent=0.0132, tangentMode=RCTM_User, time=0.5, value=1}
         {arriveTangent=-3.208, tangentMode=RCTM_User, time=1, value=0}
        ]
       }}}
     Scale Curve=&Multiply_Float(
      A=&FN_FrontEndFloatModifier(Gameplay=1, Front End=0.75)
      B=&RandomRangeFloat(Minimum=0.2, Maximum=0.3)
     )
    )
    ScaleA=true
    ScaleRGB=true
   )
   &ScaleColor(
    Color Value To Scale=$Particles.Color
    Scale Mode=::NewEnumerator1
    Scale RGBA=&MakeVector4_FromLinearColor(LinearColor=&FNTimeOfDayColor(
      Night Color=<rgba 0.539 0.6941 1 1>
      Morning Color=<rgba 0.9392 0.8785 1 1>
      Day Color=<rgba 1 1 1 1>
      Evening Color=<rgba 0.9635 0.8536 0.8431 1>
     ))
    ScaleRGBA=true
   ) {module=ScaleColor001}
   &SubUVAnimation(
    Output Alternate Renderer Binding=true
    "Sprite Renderer "=DataInterface() {propertyValues {SpriteRenderer=&self:Mist_6.NiagaraSpriteRendererProperties_0}}
    Start Frame Offset:NiagaraInt32=&RandomRangeInt(Minimum=0, Maximum=13)
    SubUV Lookup Index Scale=0.75
    UseEndFrame=false
    UseStartFrame=false
   )
   /SetVariables(Particles.Galaxy Color:LinearColor=&MakeLinearColorFromVectorAndFloat(
     "Vector (RGB)"=<1 1 1>
     "Float (Alpha)"=&FloatFromCurve(
      FloatCurve=DataInterface() {propertyValues {Curve {
         defaultValue=1
         keys [{interpMode=RCIM_Cubic, tangentMode=RCTM_User, time=0.1, value=1}, {interpMode=RCIM_Cubic, tangentMode=RCTM_User, time=0.4, value=0}]
        }}}
      Scale Curve=&FN_FrontEndFloatModifier(Gameplay=0.5, Front End=0.35)
     )
    )) {isSetParameters=true}
  ]
  rendererClasses [&NiagaraSpriteRendererProperties]
  renderers [
   {rendererClass=&NiagaraSpriteRendererProperties}
   {rendererClass=&NiagaraSpriteRendererProperties, rendererIndex=1}
  ]
 }
 Sparks {
  emitterUpdateScript [
   &EmitterState()
   &SpawnBurst_Instantaneous(Spawn Count=&UniformRangedInt(Minimum=18, Maximum=22), Spawn Time=0.02, Use Spawn Probability=false)
  ]
  particleSpawnScript [
   &InitializeParticle(
    Lifetime Max=0.4
    Lifetime Min=0.2
    Lifetime Mode=::NewEnumerator1
    Sprite Size Max=<3.5 10>
    Sprite Size Min=<2.5 6>
    Sprite Size Mode=::NewEnumerator2
    UsePositionOffset=false
    Write Material Random=true
   )
   &FN_SetSpawnGroup(Number of Spawn Groups=3)
   &CameraOffset(Camera Offset Amount=&FN_FrontEndFloatModifier(Gameplay=10, Front End=$System.FrontendCameraOffset))
   &ShapeLocation(
    Cylinder Height=&Multiply_Float(
     A=&MaskFloat(Mask Source=0, Mask Parameter=?, Mask=0, Masked Float=20, Passthrough Float=50)
     B=&FN_FrontEndFloatModifier(Gameplay=1, Front End=0.6)
    )
    Cylinder Height Midpoint:NiagaraFloat=&MaskFloatBySpawnGroup(Masked Float=0.8, Passthrough Float=0.2)
    Cylinder Radius=5
    MaskSpawnGroup=false
    Override Seed=false
    Rotation Axis=$System.DirectionVector
    Rotation Mode=::NewEnumerator1
    Shape Primitive=::NewEnumerator1
    Surface Only=false
    UseOwnerScale=true
   )
   &AddVelocity(
    Cone Angle=&MaskFloatBySpawnGroup(Masked Float=30, Passthrough Float=45)
    Cone Axis=$System.DirectionVector
    Inner Cone Angle=20
    Module.UseConeFalloff=false
    Module.UseVelDistribution=true
    Override Seed=false
    Velocity Mode=::NewEnumerator2
    Velocity Speed=&Multiply_Float(
     A=&MaskFloatBySpawnGroup(
      Masked Float=&RandomRangeFloat(Minimum=-50, Maximum=-200)
      Passthrough Float=&RandomRangeFloat(Minimum=100, Maximum=500)
     )
     B=&FN_FrontEndFloatModifier(Gameplay=1, Front End=0.6)
    )
   ) {module=AddVelocity001}
  ]
  particleUpdateScript [
   &ParticleState()
   &ScaleColor(
    Scale Mode=::NewEnumerator1
    Scale RGBA=&MakeVector4_FromLinearColor(LinearColor=&VariantColorSelector(
      Variant Amount=4
      Original Color=&ColorFromCurve(ColorCurve=DataInterface() {propertyValues {
         AlphaCurve {keys [<at 0 1>,<at 1 0>]}
         BlueCurve {keys [<at 0 3>,<at 0.4727 1>,<at 0.8261 0.3391>]}
         GreenCurve {keys [<at 0 2.0629>,<at 0.4727 0.4014>,<at 0.8261 0.1233>]}
         RedCurve {keys [<at 0 0.887>,<at 0.4727 0.2087>,<at 0.8261 0.0885>]}
        }})
      Variant 1 Color=&ColorFromCurve(ColorCurve=DataInterface() {propertyValues {
         AlphaCurve {keys [<at 0 1>,<at 1 0>]}
         BlueCurve {keys [<at 0 3>,<at 0.5249 1>,<at 1 0.3391>]}
         GreenCurve {keys [<at 0 1.063>,<at 0.5249 0.5>,<at 1 0.0885>]}
         RedCurve {keys [<at 0 0.887>,<at 0.5249 0.5417>,<at 1 0.172>]}
        }})
      Variant 2 Color=&ColorFromCurve(ColorCurve=DataInterface() {propertyValues {
         AlphaCurve {keys [<at 0 1>,<at 1 0>]}
         BlueCurve {keys [<at 0 0.6>,<at 0.4345 0.2>,<at 1 0>]}
         GreenCurve {keys [<at 0 1.8522>,<at 0.4345 0.5333>,<at 1 0.2>]}
         RedCurve {keys [<at 0 3>,<at 0.4345 1>,<at 1 0.6>]}
        }})
      Variant 3 Color=&MultiplyLinearColor(
       Linear Color=&SelectLinearColorFromArray(
        Array Sampling Mode=::NewEnumerator0
        Direct Array Index=?
        Color Selection Array=DataInterface() {propertyValues {ColorData [<rgba 2.4403 0.9652 3 1>, <rgba 0.9844 1.37 3 1>, <rgba 0.913 2.7278 3 1>, <rgba 3 1.0957 2.3873 1>]}}
        Random Seed=0
       )
       Scale Factor=&MakeVector4_FromFloat(Float=&FloatFromCurve(FloatCurve=DataInterface() {propertyValues {Curve {defaultValue=1, keys [{tangentMode=RCTM_User, time=0, value=1}, {tangentMode=RCTM_User, time=1, value=0}]}}}))
      )
      Variant 4 Color=&ColorFromCurve(ColorCurve=DataInterface() {propertyValues {
         AlphaCurve {keys [<at 0 1>,<at 1 0>]}
         BlueCurve {keys [<at 0 3>,<at 0.5062 1>,<at 1 0.3391>]}
         GreenCurve {keys [<at 0 1.575>,<at 0.5062 0.4014>,<at 1 0.1233>]}
         RedCurve {keys [<at 0 0.15>,<at 0.5062 0.2087>,<at 1 0.0885>]}
        }})
      Variant 5 Color=&ColorFromCurve(ColorCurve=DataInterface() {propertyValues {
         AlphaCurve {keys [<at 0 1>,<at 1 0>]}
         BlueCurve {keys [<at 0 3>,<at 0.8063 1>]}
         GreenCurve {keys [<at 0 0.6189>,<at 0.8063 0>]}
         RedCurve {keys [<at 0 0>,<at 0.8063 0.6435>]}
        }})
     ))
    ScaleRGBA=true
   )
   &ScaleSpriteSize(Scale Factor=&Vector2DFromFloat(Value=&FloatFromCurve(
      FloatCurve=DataInterface() {propertyValues {Curve {keys [
          {arriveTangent=10.1867, interpMode=RCIM_Cubic, leaveTangent=10.1867, tangentMode=RCTM_User, time=0, value=0}
          {interpMode=RCIM_Cubic, leaveTangent=-0.0263, tangentMode=RCTM_Break, time=0.2, value=1.2}
          {arriveTangent=-2.7631, interpMode=RCIM_Cubic, leaveTangent=-2.7631, tangentMode=RCTM_User, time=1, value=0}
         ]}}}
      Scale Curve=&Multiply_Float(A=0.4, B=&FN_FrontEndFloatModifier(Gameplay=1, Front End=0.6))
     ))) {module=ScaleSpriteSize001}
   &CurlNoiseForce(Noise Strength=200, Pan Noise=true, Pan Noise Field=<0 0 0.5>, Random Seed=10)
   &Drag(Drag=3, Use Linear Drag=true, Use Rotational Drag=false)
   &SolveForcesAndVelocity(Clamp Velocity=false, Limit Acceleration=false)
   &ScaleSpriteSizeBySpeed(Max Scale Factor=<0.8 3>, Velocity Threshold=100)
   /SetVariables(Particles.Glow03_Color=&MakeLinearColorFromVectorAndFloat("Vector (RGB)"=&MakeVectorFromLinearColorRGB(Color=$Particles.Color), "Float (Alpha)"=0.14)) {isSetParameters=true}
   /SetVariables(Particles.Glow03_Size=&Multiply_Vector2D(A=$Particles.SpriteSize, B=<4 2>)) {isSetParameters=true}
  ]
  rendererClasses [&NiagaraSpriteRendererProperties]
  renderers [
   {rendererClass=&NiagaraSpriteRendererProperties}
   {rendererClass=&NiagaraSpriteRendererProperties, rendererIndex=1}
  ]
 }
 SpikeStreaksParticle {
  emitterUpdateScript [
   &EmitterState()
   &SpawnBurst_Instantaneous(
    Spawn Count=&RandomRangeInt(Minimum=6, Maximum=8)
    Use Spawn Probability=false
    UseLoopCountLimit=false
   )
  ]
  particleSpawnScript [
   &InitializeParticle(
    Color=&VariantColorSelector(
     Variant Amount=4
     Original Color=<rgba 0.2291 0.4904 0.7454 1>
     Variant 1 Color=<rgba 0.35 0.35 1 1>
     Variant 2 Color=<rgba 1 0.6266 0.1792 1>
     Variant 3 Color=&SelectLinearColorFromArray(
      Array Sampling Mode=::NewEnumerator0
      Direct Array Index=?
      Color Selection Array=DataInterface() {propertyValues {ColorData [<rgba 0.9216 0.347 1 1>, <rgba 0.2658 0.1457 1 1>, <rgba 0.202 0.4019 1 1>, <rgba 0.0896 0.8566 1 1>]}}
      Random Seed=0
     )
     Variant 4 Color=<rgba 0 0.425 0.85 1>
     Variant 5 Color=&SelectLinearColorFromArray(
      Array Sampling Mode=::NewEnumerator0
      Direct Array Index=?
      Color Selection Array=DataInterface() {propertyValues {ColorData [<rgba 0.478 0.1411 1.2 1>, <rgba 0.165 0.165 1.1 1>]}}
      Random Seed=0
     )
    )
    Lifetime Max=0.2087
    Lifetime Min=0.05
    Lifetime Mode=::NewEnumerator1
    Mesh Scale Mode=::NewEnumerator4
    Sprite Size=&Multiply_Vector2D_ByFloat(
     Float=&FN_FrontEndFloatModifier(Gameplay=1, Front End=0.3)
     Vector2D=&RandomRangeVector2D(Minimum=<10 75>, Maximum=<6 50>)
    )
    Sprite Size Mode=::NewEnumerator4
    UsePositionOffset=false
    Write Material Random=true
   )
   &CameraOffset(Camera Offset Amount=25)
   &AddVelocity(
    Cone Angle=90
    Cone Axis=$System.DirectionVector
    Module.UseConeFalloff=false
    Module.UseVelDistribution=false
    Override Seed=false
    Velocity Mode=::NewEnumerator2
   )
  ]
  particleUpdateScript [
   &ParticleState()
   &SolveForcesAndVelocity(Clamp Velocity=false, Limit Acceleration=false)
   &ScaleSpriteSize(Uniform Curve Scale=1.4)
   &ScaleColor(
    Scale Alpha=&FloatFromCurve(FloatCurve=DataInterface() {propertyValues {Curve {keys [<at 0.7 1>,<at 1 0>]}}}, Scale Curve=1.5)
    Scale RGB=&MakeVectorFromLinearColorRGB(Color=&Multiply_LinearColorByFloat(Linear Color=$System.TODColor, Float=$System.TODFloat))
    ScaleA=true
    ScaleRGB=true
   )
  ]
  rendererClasses [&NiagaraSpriteRendererProperties]
  renderers [{rendererClass=&NiagaraSpriteRendererProperties}]
 }
 Variant_Elements {
  emitterUpdateScript [
   &EmitterState()
   &SpawnBurst_Instantaneous(Spawn Count=&RandomRangeInt(Minimum=3, Maximum=4), Use Spawn Probability=true, UseLoopCountLimit=false)
  ]
  particleSpawnScript [
   &InitializeParticle(
    Color=&MergeHelper_ColorSwitch(
     0=<rgba 0 0 0 0>
     1=&ColorFromCurve(
      ColorCurve=DataInterface() {propertyValues {
        AlphaCurve {keys [<at 0 1 RCIM_Cubic>]}
        BlueCurve {keys [<at 0.4274 0.9047 RCIM_Cubic>,<at 0.4699 0.6751 RCIM_Cubic>]}
        GreenCurve {keys [<at 0.4274 0.8796 RCIM_Cubic>,<at 0.4699 0.6696 RCIM_Cubic>]}
        RedCurve {keys [<at 0.4274 1 RCIM_Cubic>,<at 0.4699 1 RCIM_Cubic>]}
       }}
      CurveIndex=&RandomRangeFloat(Minimum=0, Maximum=1)
     )
     2=&ColorFromCurve(
      ColorCurve=DataInterface() {propertyValues {
        AlphaCurve {keys [<at 0 1 RCIM_Cubic>]}
        BlueCurve {keys [<at 0.4274 0 RCIM_Cubic>,<at 0.4699 0 RCIM_Cubic>]}
        GreenCurve {keys [<at 0.4274 0.6232 RCIM_Cubic>,<at 0.4699 0.3648 RCIM_Cubic>]}
        RedCurve {keys [<at 0.4274 1 RCIM_Cubic>,<at 0.4699 1 RCIM_Cubic>]}
       }}
      CurveIndex=&RandomRangeFloat(Minimum=0, Maximum=1)
     )
     3=&ColorFromCurve(
      ColorCurve=DataInterface() {propertyValues {
        AlphaCurve {keys [<at 0 1 RCIM_Cubic>]}
        BlueCurve {keys [<at 0.3184 1 RCIM_Cubic>,<at 0.3628 1 RCIM_Cubic>,<at 0.7195 1>,<at 0.737 0.7715>]}
        GreenCurve {keys [
          <at 0.3184 0.6924 RCIM_Cubic>
          {arriveTangent=-1.1028, interpMode=RCIM_Cubic, leaveTangent=-1.1028, time=0.3628, value=0.3664}
          <at 0.7195 0.25>
          <at 0.737 1>
         ]}
        RedCurve {keys [
          <at 0.3184 0.2 RCIM_Cubic>
          {arriveTangent=1.3543, interpMode=RCIM_Cubic, leaveTangent=1.3543, time=0.3628, value=0.7831}
          <at 0.7195 0.7433>
          <at 0.737 0.2>
         ]}
       }}
      CurveIndex=&RandomRangeFloat(Minimum=0, Maximum=1)
     )
     4=&ColorFromCurve(
      ColorCurve=DataInterface() {propertyValues {
        AlphaCurve {keys [<at 0 1 RCIM_Cubic>]}
        BlueCurve {keys [<at 0.2316 1 RCIM_Cubic>,<at 0.8299 1>]}
        GreenCurve {keys [<at 0.2316 0.8118 RCIM_Cubic>,<at 0.8299 0.6913>]}
        RedCurve {keys [<at 0.2316 0.5104 RCIM_Cubic>,<at 0.8299 0.5131>]}
       }}
      CurveIndex=&RandomRangeFloat(Minimum=0, Maximum=1)
     )
     5=&ColorFromCurve(
      ColorCurve=DataInterface() {propertyValues {
        AlphaCurve {keys [<at 0 1 RCIM_Cubic>]}
        BlueCurve {keys [<at 0.4274 1 RCIM_Cubic>,<at 0.4699 1 RCIM_Cubic>]}
        GreenCurve {keys [<at 0.4274 0 RCIM_Cubic>,<at 0.4699 0.282 RCIM_Cubic>]}
        RedCurve {keys [<at 0.4274 0.6698 RCIM_Cubic>,<at 0.4699 0 RCIM_Cubic>]}
       }}
      CurveIndex=&RandomRangeFloat(Minimum=0, Maximum=1)
     )
     Index=&MakeIntFromFloat(Float=$User.Variant)
     Num=::NewEnumerator4
    )
    Lifetime=&RandomRangeFloat(Minimum=0.3, Maximum=0.6)
    Mesh Scale Mode=::NewEnumerator1
    Mesh Uniform Scale Max=0.4
    Mesh Uniform Scale Min=0.3
    Sprite Rotation Angle=&MergeHelper_FloatSwitch(
     0=0
     1=0
     2=&RandomRangeFloat(Minimum=0, Maximum=1)
     3=&RandomRangeFloat(Minimum=0, Maximum=1)
     4=0
     5=&RandomRangeFloat(Minimum=0, Maximum=1)
     Index=&MakeIntFromFloat(Float=$User.Variant)
     Num=::NewEnumerator4
    )
    Sprite Rotation Mode=::NewEnumerator1
    Sprite Size Mode=::NewEnumerator3
    Uniform Sprite Size=&Multiply_Float(
     A=&Multiply_Float(
      A=&RandomRangeFloat(Minimum=10, Maximum=14)
      B=&FN_FrontEndFloatModifier(Gameplay=1, Front End=0.6)
     )
     B=&SelectFloatFromArray(
      Direct Array Index=&MakeIntFromFloat(Float=$User.Variant)
      Float Selection Array=DataInterface() {propertyValues {FloatData [0,0,1.5,0.8,1,1.5]}}
     )
    )
    UsePositionOffset=true
    Write Material Random=false
   )
   &ShapeLocation(
    MaskSpawnGroup=false
    Override Seed=false
    Sphere Radius=&Multiply_Float(
     A=&Multiply_Float(A=10, B=&FN_FrontEndFloatModifier(Gameplay=1, Front End=0.6))
     B=&SelectFloatFromArray(
      Direct Array Index=&MakeIntFromFloat(Float=$User.Variant)
      Float Selection Array=DataInterface() {propertyValues {FloatData [1,1,1,1,1,2]}}
     )
    )
    Transform Order:ENiagara_TransformOrder=::NewEnumerator1
    UseOwnerScale=true
   ) {module=ShapeLocation001}
   &CameraOffset(Camera Offset Amount=&FN_FrontEndFloatModifier(Gameplay=15, Front End=0)) {module=CameraOffset001}
   &AddVelocity(
    Cone Angle=&RandomRangeFloat(Minimum=30, Maximum=35)
    Cone Axis=$System.DirectionVector
    Module.UseConeFalloff=false
    Module.UseVelDistribution=false
    Override Seed=false
    Velocity Mode=::NewEnumerator2
    Velocity Speed=&Multiply_Float(
     A=&RandomRangeFloat(Minimum=200, Maximum=600)
     B=&FN_FrontEndFloatModifier(Gameplay=1, Front End=0.6)
    )
   )
   &InitialMeshOrientation(
    Mesh Orientation Mode=::NewEnumerator2
    Orientation Vector=$System.DirectionVector
    Rotation=&RandomRangeVector(Minimum=<0 0 0>, Maximum=<0 0 1>)
    Use Rotation Vector=true
   )
   /SetVariables(Particles.VisibilityTag=&MakeIntFromFloat(Float=$System.VariantIndex)) {isSetParameters=true}
  ]
  particleUpdateScript [
   &ParticleState()
   &AccelerationForce(Acceleration=&Multiply_VectorByFloat(
     Vector=&RandomRangeVector(Minimum=<0 0 -100>, Maximum=<0 0 -200>)
     Float=&FN_FrontEndFloatModifier(Gameplay=1, Front End=0.6)
    ))
   &Drag(Drag=&RandomRangeFloat(Minimum=6, Maximum=8), Use Linear Drag=true, Use Rotational Drag=true)
   &CurlNoiseForce(Bias Noise=false, Noise Strength=&RandomRangeFloat(Minimum=80, Maximum=120), Pan Noise=false)
   &SolveForcesAndVelocity(Clamp Velocity=false, Limit Acceleration=false)
   &ScaleSpriteSize(Uniform Curve Sprite Scale=DataInterface() {propertyValues {Curve {keys [
        {arriveTangent=9.0725, interpMode=RCIM_Cubic, leaveTangent=3.7536, tangentMode=RCTM_Break, time=0, value=0}
        <at 0.5 1 RCIM_Cubic>
        {arriveTangent=-0.0405, interpMode=RCIM_Cubic, leaveTangent=-0.0405, tangentMode=RCTM_User, time=1, value=0}
       ]}}})
   &SpriteRotationRate(Rotation Rate=&FloatFromCurve(
     FloatCurve=DataInterface() {propertyValues {Curve {keys [{interpMode=RCIM_Cubic, tangentMode=RCTM_User, time=0, value=1}, {interpMode=RCIM_Cubic, tangentMode=RCTM_User, time=1, value=0}]}}}
     Scale Curve=&Multiply_FloatByInt(
      Float=&MergeHelper_FloatSwitch(
       0=0
       1=0
       2=&RandomRangeFloat(Minimum=500, Maximum=600)
       3=&RandomRangeFloat(Minimum=200, Maximum=300)
       4=0
       5=&RandomRangeFloat(Minimum=500, Maximum=600)
       Index=&MakeIntFromFloat(Float=$User.Variant)
       Num=::NewEnumerator4
      )
      Integer=&SelectIntFromArray(
       Array Sampling Mode=::NewEnumerator1
       Direct Array Index=?
       Int Selection Array=DataInterface() {propertyValues {IntData [-1,1]}}
       Random Seed=0
      )
     )
    ))
  ]
  rendererClasses [&NiagaraSpriteRendererProperties]
  renderers [
   {rendererClass=&NiagaraSpriteRendererProperties}
   {rendererClass=&NiagaraSpriteRendererProperties, rendererIndex=1}
   {rendererClass=&NiagaraSpriteRendererProperties, rendererIndex=2}
  ]
 }
}
summary {
 systemName=NS_Pickaxe_FrozenTouch_Impact
 emitters [
  {emitterName=SpikeStreaksParticle, rendererClasses [&NiagaraSpriteRendererProperties]}
  {emitterName=Mist, rendererClasses [&NiagaraSpriteRendererProperties]}
  {emitterName=Ice_Shards, rendererClasses [&NiagaraMeshRendererProperties]}
  {emitterName=Directional_Flash, rendererClasses [&NiagaraSpriteRendererProperties]}
  {emitterName=FlashElements, rendererClasses [&NiagaraSpriteRendererProperties, &NiagaraLightRendererProperties]}
  {emitterName=Sparks, rendererClasses [&NiagaraSpriteRendererProperties]}
  {emitterName=Variant_Elements, rendererClasses [&NiagaraSpriteRendererProperties]}
 ]
 userVariables [
  User.bIsFrontEnd {type=NiagaraBool}
  User.bIsFrontEndPreview {type=NiagaraBool}
  User.SwingSectionIndex {type=NiagaraInt32}
  User.Variant {type=NiagaraFloat, defaultValue=0, description=""}
 ]
}
